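# Translucent unlit ice-cube material for Freeze Trap FX mesh particles, likely the ice around frozen feet, with index-of-refraction distortion.
# Fakes lighting with a tinted skylight cubemap and sun highlight, added to tinted diffuse, Fresnel terms and a depth-based intersection edge.
# A dynamic Crack parameter thresholds a noise texture through Floor, revealing cracks in normals and color.
# A WobbleTime timestamp drives an eased, sine-based wobble along vertex normals.
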Material /Game/Athena/Items/Traps/FreezeTrap/FX/M_Ice_Feet_Cube
graph {
 properties {
  bUsedWithMeshParticles=true
  bUsedWithNiagaraMeshParticles=true
  blendMode=BLEND_Translucent
  pixelDepthOffsetMode=PDOM_Legacy
  refractionMethod=RM_IndexOfRefraction
  shadingModel=MSM_Unlit
  usedShadingModels=Unlit
 }
 nodes [
  n1 = TextureCoordinate() {uTiling=1.5, vTiling=1.5}
  n2 = TextureSample(UVs=@n1) {texture=/Game/Effects/Fort_Effects/Textures/Tile/T_Glass_Cracks_M}
  n3 = Multiply(A=@n2.R) {constB=0.5}
  n4 = Multiply(
   A=TextureSample(UVs=@n1) {samplerType=SAMPLERTYPE_Normal, texture=/Game/Effects/Fort_Effects/Textures/Tile/T_Glass_Cracks_N}.RGB
   B=Constant3Vector() {constant=<rgba 3 3 0 0>}
  )
  n5 = DynamicParameter() {defaultValue=<rgba 0 1 1 1>, paramNames [Crack,WobbleTime,Param3,Param4]}
  n6 = Floor(None=Add(A=TextureSample(UVs=@n1) {texture=&T_Smooth_tile}.G, B=@n5.Crack))
  Add(
   A=TextureSample() {samplerType=SAMPLERTYPE_Normal, texture=&/T_IceCube_Cheap_N}.RGB
   B=Multiply(A=Add(A=@n3, B=@n4), B=@n6)
  )
  n7 = MaterialFunctionCall!(
   Distance Bias=ScalarParameter() {defaultValue=5.7607, group=IntersectingEdge, parameterName=IntersectingEdgeBias}
   Distance Divisor=ScalarParameter() {defaultValue=57.56, group=IntersectingEdge, parameterName=IntersectingEdgeDivisor}
   Scene Depth
   Pixel Depth
  ) {materialFunction=/Game/Effects/Fort_Effects/Materials/_Functions/MF_DistanceBehindTranslucency_01}
  n8 = Add(
   A=Multiply(
    A=Saturate(None=@n7.Normalized Inverted Unclamped)
    B=ScalarParameter() {defaultValue=0.3012, group=IntersectingEdge, parameterName=Intersection Edge Large}
   )
   B=Power(
    Base=@n7.Normalized Inverted Unclamped
    Exp=ScalarParameter() {defaultValue=122.7034, group=IntersectingEdge, parameterName=Intersection Edge Exponent}
   )
  )
  n9 = TextureSampleParameterCube!(UVs=ReflectionVectorWS() {bNormalizeCustomWorldNormal=true}, Apply View MipBias) {
   parameterName=MaterialExpressionTextureSampleParameterCube_0
   samplerType=SAMPLERTYPE_LinearColor
   texture=/Game/Packages/Fortress_Sky/TexturesHDR/T_AthenaSkylight
  }
  n10 = Multiply(A=@n9.RGB, B=VectorParameter() {parameterName=Cube Map Multiplier, defaultValue=<rgba 0.3282 0.5416 0.7 0>}.RGB) {constB=1.5}
  n11 = ComponentMask(None=MaterialFunctionCall() {materialFunction=/Engine/Functions/GetPostProcessFunctions/GetAmbientCubemapTint}.AmbientCubemapTint) {b=true, g=true, r=true}
  n12 = ShadingPathSwitch(Default=Desaturation(None=Multiply(A=@n10, B=@n11), Fraction=Constant() {r=0.5}), Mobile=Constant()) {desc=Migrated from FeatureLevelSwitch}
  n13 = CollectionParameter() {
   collection=&FortniteMaterialParameters
   parameterName=SunAndMoonModelDirectionalVector
  }
  n14 = DotProduct(A=ComponentMask(None=@n13) {b=true, g=true, r=true}, B=ReflectionVectorWS())
  n15 = Multiply(
   A=Power(Base=Clamp(None=@n14), Exp=ScalarParameter() {defaultValue=70, parameterName=Spec Highlight Power}) {constExponent=20}
   B=ScalarParameter() {defaultValue=3, parameterName=Spec Intense}
  )
  n16 = TextureSampleParameter2D() {parameterName=Diffuse, texture=&/T_IceCube_Cheap_D}
  n17 = Clamp(None=LinearInterpolate(Alpha=@n16.B) {constA=0.72, constB=-3})
  n18 = ScalarParameter() {defaultValue=1, parameterName=DiffuseMultCubemap}
  n19 = Add(A=@n16.RGB) {constB=0.1}
  n20 = NamedRerouteUsage()
  n21 = Add(
   A=VectorParameter() {parameterName=BaseColor, defaultValue=<rgba 0.3121 0.4699 0.6 0>}.RGB
   B=Multiply(A=VectorParameter() {parameterName=EdgeGlow, defaultValue=<rgba 0.6583 1.3266 1.6 0>}.RGB, B=Power(Base=@n20))
  )
  n22 = OneMinus(None=@n2.G)
  n23 = Add(
   A=Multiply(A=Add(A=@n12, B=@n15), B=LinearInterpolate(B=@n17, Alpha=@n18) {constA=1})
   B=Multiply(A=Multiply(A=@n19, B=@n21), B=LinearInterpolate(B=@n22, Alpha=@n6) {constA=1})
  )
  n24 = Multiply(A=@n23, B=ParticleColor().RGB)
  n25 = ScalarParameter() {parameterName=UseParticleColorTint}
  Add(
   A=NamedRerouteDeclaration(None=Saturate(None=@n8)) {name=Intersecting Edge, nodeColor=<rgba 0 1 0.7891 1>}
   B=MaterialFunctionCall(Emissive=LinearInterpolate(A=@n23, B=@n24, Alpha=@n25)) {materialFunction=&MF_FX_EmissiveAdjustments}.Result
  )
  Add()
  Constant()
  Fresnel() {baseReflectFraction=0.4, exponent=0.5}
  Fresnel() {baseReflectFraction=0, exponent=2}
  n26 = Add(A=Sine(None=Time()) {period=0.09}) {constB=0.2}
  n27 = ScalarParameter() {defaultValue=2.2, parameterName=WobbleAmount}
  n28 = Subtract(A=Time() {period=5}, B=@n5.WobbleTime)
  n29 = ScalarParameter() {defaultValue=0.64, group=Timestamp, parameterName=FlashDuration}
  n30 = MaterialFunctionCall!(In=Clamp(None=Divide(A=@n28, B=@n29))) {materialFunction=/Game/Effects/Fort_Effects/Materials/_Functions/MF_Smoothstep_EaseOut}
  LinearInterpolate(B=Multiply(A=VertexNormalWS(), B=Multiply(A=@n26, B=@n27)), Alpha=Clamp(None=Multiply(A=@n30.Result) {constB=1.5})) {constAlpha=1}
  MaterialFunctionCall() {materialFunction=&DebugTimeSine}
  n31 = DotProduct(A=CameraVectorWS(), B=PixelNormalWS())
  NamedRerouteDeclaration(None=OneMinus(None=Abs(None=@n31) {desc=Fix for a bug where the source object has an inverted scale})) {name=Fixed Fresnel, nodeColor=<rgba 0.9219 0 1 1>}
  n32 = NamedRerouteUsage()
  n33 = ScalarParameter() {defaultValue=0.1, parameterName=Fresnel Exponent}
  Saturate(None=Power(Base=Power(Base=@n32, Exp=@n33) {constExponent=0.1}, Exp=ScalarParameter() {defaultValue=1, parameterName=Fresnel Power}) {constExponent=1, desc=Second power remains so it doesn't break how it worked before.})
  ScalarParameter() {defaultValue=0.97, parameterName=Refraction}
 ]
}
