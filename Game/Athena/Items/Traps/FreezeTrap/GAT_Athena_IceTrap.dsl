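; Gameplay ability for the Athena freeze trap that clears the Icy Feet gameplay effect from its owner.
; On activation it calls the parent, runs the generic ability setup and stores the returned player pawn.
; It then removes owner effects carrying the GameplayEffect.Trap.Ice.IcyFeet asset tag and immediately ends.
; The end-ability event has no logic, and nothing here applies a freeze.
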(event Ability|EventActivateAbility
  (|CallParentFunction)
  (bind (_event_data _fort_player_pawn) (Setup|SetupGenericAbility false false)
    (:Output
      (Variables|Default|SetPlayerPawn _fort_player_pawn)
      (Ability|RemoveGameplayEffectFromOwnerWithAssetTags "(GameplayTags=((TagName=\"GameplayEffect.Trap.Ice.IcyFeet\")))")
      (Ability|EndAbility))))

(event Ability|EventOnEndAbility
  (params (bWasCancelled :type Boolean)))
