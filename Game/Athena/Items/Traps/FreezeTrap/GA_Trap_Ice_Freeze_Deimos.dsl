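; Gameplay ability for the Chiller trap that freezes an AI-controlled avatar in place while active.
; On activation it targets the avatar's Fort AI pawn and controller, setting the pawn's frozen flag, stopping brain logic and halting movement.
; On ability end it clears the flag and restarts the brain logic.
; The ability never ends itself, so the freeze duration is controlled outside this graph.
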(event Ability|EventActivateAbility
  (bind _returnvalue (Ability|GetAvatarActorfromActorInfo))
  (bind _returnvalue_2 (AI|GetAIController _returnvalue))
  (bind _asfort_aipawn (Utilities|Casting|CastToFortAIPawn _returnvalue)
    (:then
      (Variables|/Script/FortniteGame.FortAIPawn|SetbFrozen true _asfort_aipawn))
    (:CastFailed
      _))
  (AI|Logic|StopLogic (Variables|/Script/AIModule.AIController|GetBrainComponent _returnvalue_2) "Frozen by GA_Trap_Ice_Freeze_Deimos as part of the Chiller trap.")
  (AI|Navigation|StopMovement _returnvalue_2))

(event Ability|EventOnEndAbility
  (params (bWasCancelled :type Boolean))
  (bind _returnvalue (Ability|GetAvatarActorfromActorInfo))
  (bind _asfort_aipawn (Utilities|Casting|CastToFortAIPawn _returnvalue)
    (:then
      (Variables|/Script/FortniteGame.FortAIPawn|SetbFrozen false _asfort_aipawn))
    (:CastFailed
      _))
  (AI|Logic|RestartLogic (Variables|/Script/AIModule.AIController|GetBrainComponent (AI|GetAIController _returnvalue))))
