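; Athena floor freeze trap Blueprint, handling the trap's trigger registration and its audio and visual feedback.
; Placement registers the trigger and dims emissive after caching its active color; ready, firing and reload events toggle emissive, idle/active FX and sounds.
; Running out of durability plays a delayed explosion using Spike Trap destruction FX.
; SetTeamColor handles friendly-team detection, material and idle FX, but no graph here calls it, and OnInitTeam is empty.
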(event EventOnPlaced
  ;[1 1 1] Initial Setup
  (Utilities|FlowControl|DoOnce false
    (:Completed
      (Trap|AddTriggerComponent (Variables|Trap_Floor_Spikes|GetTrigger))
      (Networking|CanExecuteCosmeticEvents
        (:True
          (Audio|PlaySoundatLocation (Variables|Audio|GetPlace_Trap_Sound) (Transformation|GetActorLocation))
          (Utilities|TrapPlacementDustFX :Attach%20to%20Component (Variables|Debug|GetStaticMeshComponent) :Particle%20System%20%28Override%29 (Variables|Default|GetPS_DustPlacement) :Socket "PlacedFX" :Rotation "0, 0, 0" :Burst%20Rate 15 :Size%20%28Volume%29 "256.000000,256.000000,15.000000" :Velocity%20Scale "1,1,1" :Size%20Scale%20%28Particles%29 1
            (:Out
              (bind _returnvalue (CallFunction|GetorCreateMIDOnDefaultMesh 0))
              (bind _returnvalue_2 (Rendering|Material|GetVectorParameterValue _returnvalue "EmissiveMultiplier" :!class /Script/Engine.MaterialInstanceDynamic))
              (Variables|Default|SetActiveEmissiveColor _returnvalue_2)
              (Utilities|FlowControl|Delay 0.2)
              (Rendering|Material|SetVectorParameterValueonMaterials (Variables|Debug|GetStaticMeshComponent) "EmissiveMultiplier" (Math|Conversions|ToVector(LinearColor) (Variables|Default|GetInactiveEmissiveColor))))))))))

(event EventOnFinishedBuilding
  ;[1 1 1] Finished building
  (Networking|CanExecuteCosmeticEvents
    (:True
      (Audio|PlaySoundatLocation (Variables|Audio|GetTrap_Active_Sound) (Transformation|GetActorLocation))
      (Rendering|Material|SetVectorParameterValueonMaterials (Variables|Debug|GetStaticMeshComponent) "EmissiveMultiplier" (Math|Conversions|ToVector(LinearColor) (Variables|Default|GetActiveEmissiveColor)))
      (Components|Activation|SetActive (Variables|Default|GetIdleFX) true false))))

(event EventOnWorldReady
  ;[1 1 1] Finished building
  (Networking|CanExecuteCosmeticEvents
    (:True
      (Audio|PlaySoundatLocation (Variables|Audio|GetTrap_Active_Sound) (Transformation|GetActorLocation))
      (Rendering|Material|SetVectorParameterValueonMaterials (Variables|Debug|GetStaticMeshComponent) "EmissiveMultiplier" (Math|Conversions|ToVector(LinearColor) (Variables|Default|GetActiveEmissiveColor)))
      (Components|Activation|SetActive (Variables|Default|GetIdleFX) true false))))

(event GameplayCue|GameplayCue.Abilities.Activation.Traps.ActivateTrap
  (params (EventType :type EGameplayCueEvent%20Enum :default "OnActive") (Parameters :type Gameplay%20Cue%20Parameters%20Structure%20%28by%20ref%29))
  ;[1 1 1] On Fire
  (switch Utilities|FlowControl|Switch|SwitchonEGameplayCueEvent EventType
    (:Executed
      (Audio|PlaySoundatLocation (Variables|Audio|GetTrap_Fire_Sound) (Transformation|GetActorLocation))
      (Components|Activation|SetActive (Variables|Default|GetActiveFX) true true)
      (Components|Activation|SetActive (Variables|Default|GetIdleFX) false false)
      (Utilities|FlowControl|Delay 0.5)
      (Rendering|Material|SetVectorParameterValueonMaterials (Variables|Debug|GetStaticMeshComponent) "EmissiveMultiplier" (Math|Conversions|ToVector(LinearColor) (Variables|Default|GetInactiveEmissiveColor))))))

(event EventOnReloadEnd
  ;[1 1 1] Reload End
  (Audio|PlaySoundatLocation (Variables|Audio|GetTrap_Active_Sound) (Transformation|GetActorLocation))
  (Components|Activation|SetActive (Variables|Default|GetIdleFX) true false)
  (Rendering|Material|SetVectorParameterValueonMaterials (Variables|Debug|GetStaticMeshComponent) "EmissiveMultiplier" (Math|Conversions|ToVector(LinearColor) (Variables|Default|GetActiveEmissiveColor))))

(event EventOnOutOfDurability
  ;[1 1 1] Durability Change
  (Utilities|FlowControl|Delay 1)
  (Audio|PlaySoundatLocation (Variables|Audio|GetTrap_Explode_Sound) (Transformation|GetActorLocation))
  (Effects|SpawnEmitteratLocation "/Game/Weapons/FORT_TRAPS/Spike_Trap/P_Spike_Trap_Destruction.P_Spike_Trap_Destruction" (Transformation|GetWorldLocation (Variables|Transformation|GetRootComponent))))

(event EventOnInitTeam)

(fn SetTeamColor
  (bind _returnvalue (Fortnite|GetLocalFortPlayerControllers true))
  (if (and (Utilities|Enum|Equal(Enum) (Fortnite|GetSubGame) "Athena") (> (Utilities|Array|Length _returnvalue) 0))
    (Variables|Default|SetFriendlyTrap (Math|Byte|Equal(Byte) (Variables|Team|GetTeamIndex) (Team|GetActorTeam (Utilities|Array|Get(acopy) _returnvalue))))
    (Rendering|Material|SetMaterial (Variables|Debug|GetStaticMeshComponent) 0 "/Game/Weapons/FORT_TRAPS/Floor_Freeze/Materials/M_Floor_Freeze_Athena_Friendly_Inst.M_Floor_Freeze_Athena_Friendly_Inst"))
  (bind _returnvalue_2 (CallFunction|GetorCreateMIDOnDefaultMesh 0))
  (bind _returnvalue_3 (Rendering|Material|GetVectorParameterValue _returnvalue_2 "EmissiveMultiplier" :!class /Script/Engine.MaterialInstanceDynamic))
  (Variables|Default|SetActiveEmissiveColor _returnvalue_3)
  (if (Variables|Default|GetFriendlyTrap)
    (Components|Activation|SetActive (Variables|Default|GetIdleFX) false false)))

(fn ConstructionScript)
